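# Unlit, two-sided translucent material for the front-end item preview plane, showing a texture on a rounded card with an animated outline.
# MF_UI_Box builds the fill mask, outline and outer glow, while the texture is inset from its center.
# Four tinted corner colors form a rotating gradient, alongside a periodic sheen from a borrowed pickaxe MF_Sheen function.
# A static switch enables sine-driven vertical floating between top and bottom distances.
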Material /BRCosmetics/Environments/FrontEnd/Blueprints/ItemPreview/Materials/M_BP_ItemPreview_Plane
graph {
 properties {
  bAllowFrontLayerTranslucency=false
  bOutputTranslucentVelocity=true
  blendMode=BLEND_Translucent
  pixelDepthOffsetMode=PDOM_Legacy
  shadingModel=MSM_Unlit
  twoSided=true
  usedShadingModels=Unlit
 }
 nodes [
  n1 = NamedRerouteUsage()
  n2 = NamedRerouteUsage()
  Add(A=NamedRerouteUsage(), B=LinearInterpolate(A=NamedRerouteUsage(), B=Add(A=@n1, B=@n2), Alpha=NamedRerouteUsage()))
  n3 = NamedRerouteUsage()
  n4 = MaterialFunctionCall(
   Size=Constant() {r=1}
   CornerRoundness=ScalarParameter() {defaultValue=0.2, group=Outline, parameterName=CornerRadius, sortPriority=2}
   OutlineThickness=ScalarParameter() {defaultValue=0.03, group=Out Glow, parameterName=OutGlowThickness, sortPriority=1}
  ) {materialFunction=&MF_UI_Box}
  n5 = OneMinus(None=SmoothStep(Value=OneMinus(None=@n4.SDFOutline)) {constMax=0.98, constMin=2})
  NamedRerouteDeclaration(None=Multiply(
    A=LinearInterpolate(B=@n3, Alpha=@n5)
    B=ScalarParameter() {defaultValue=200, group=Out Glow, parameterName=OutGlowIntensity, sortPriority=2}
   ) {constA=500, constB=200}) {name=Out Glow, nodeColor=<rgba 0 0 0 1>}
  NamedRerouteDeclaration(None=Multiply(
    A=VectorParameter() {group=Color, parameterName=Color_1, sortPriority=1, defaultValue=<rgba 1 0 0 1>}.RGB
    B=ScalarParameter() {defaultValue=1, group=Color, parameterName=IntensityColor_1, sortPriority=2}
   )) {name=Color 1, nodeColor=<rgba 0 0 0 1>}
  n6 = ComponentMask(None=Rotator(Coordinate=TextureCoordinate(), Time=ScalarParameter() {defaultValue=-10.5, group=Sheen, parameterName=SheenAngle, sortPriority=3})) {g=true}
  n7 = Time()
  n8 = ScalarParameter() {defaultValue=3, group=Sheen, parameterName=SheenDelay, sortPriority=5}
  n9 = Multiply(A=OneMinus(None=Fmod(A=@n7, B=@n8)), B=ScalarParameter() {defaultValue=1, group=Sheen, parameterName=SheenSpeed, sortPriority=4}) {constB=1.5}
  n10 = ScalarParameter() {defaultValue=3, group=Sheen, parameterName=SheenScale, sortPriority=2}
  NamedRerouteDeclaration(None=Multiply(
    A=MaterialFunctionCall!(Gradient=@n6, Mask Position=@n9, Mask Size=@n10) {materialFunction=/BRCosmetics/Weapons/FORT_Melee/Pickaxe_Dojo_Male_1H/FX/MF_Sheen}.Sinusoidal
    B=ScalarParameter() {defaultValue=0.1, group=Sheen, parameterName=SheenIntensity, sortPriority=1}
   )) {name=Sheen, nodeColor=<rgba 0 0 0 1>}
  n11 = NamedRerouteUsage()
  n12 = NamedRerouteUsage()
  n13 = MaterialFunctionCall(UV Channel=Rotator(
    Coordinate=MaterialFunctionCall() {materialFunction=&ScaleUVsByCenter}.UVs
    Time=Time()
   ) {speed=-1.5}) {materialFunction=&LinearGradient}
  n14 = NamedRerouteUsage()
  n15 = NamedRerouteUsage()
  NamedRerouteDeclaration(None=LinearInterpolate(A=LinearInterpolate(A=@n14, B=@n15, Alpha=@n13.UGradient), B=LinearInterpolate(A=@n11, B=@n12, Alpha=@n13.UGradient), Alpha=@n13.VGradient)) {name=Gradient, nodeColor=<rgba 0 0 0 1>}
  n16 = Constant()
  n17 = MaterialFunctionCall(R=@n16, G=@n16, B=ScalarParameter() {defaultValue=5, group=Animation, parameterName=FloatingTopDistance, sortPriority=2}, A=@n16) {materialFunction=&AppendMany}
  n18 = Constant()
  n19 = MaterialFunctionCall(R=@n18, G=@n18, B=ScalarParameter() {group=Animation, parameterName=FloatingBottomDistance, sortPriority=3}, A=@n18) {materialFunction=&AppendMany}
  n20 = Sine(None=Multiply(A=Time(), B=ScalarParameter() {defaultValue=0.1, group=Animation, parameterName=Floating Speed, sortPriority=4}))
  NamedRerouteDeclaration(None=StaticSwitchParameter(True=LinearInterpolate(A=@n17.RGB, B=@n19.RGB, Alpha=@n20), False=Constant()) {defaultValue=true, group=Animation, parameterName=Is Floating, sortPriority=1}) {name=Is Floating, nodeColor=<rgba 1 0 1 1>}
  NamedRerouteDeclaration(None=Multiply(
    A=VectorParameter() {group=Color, parameterName=Color_2, sortPriority=3, defaultValue=<rgba 0 1 0 1>}.RGB
    B=ScalarParameter() {defaultValue=1, group=Color, parameterName=IntensityColor_2, sortPriority=4}
   )) {name=Color 2, nodeColor=<rgba 0 0 0 1>}
  NamedRerouteDeclaration(None=Multiply(
    A=VectorParameter() {group=Color, parameterName=Color_3, sortPriority=5, defaultValue=<rgba 0 0 1 1>}.RGB
    B=ScalarParameter() {defaultValue=1, group=Color, parameterName=IntensityColor_3, sortPriority=6}
   )) {name=Color 3, nodeColor=<rgba 0 0 0 1>}
  NamedRerouteDeclaration(None=Multiply(
    A=VectorParameter() {group=Color, parameterName=Color_4, sortPriority=7, defaultValue=<rgba 1 0 1 1>}.RGB
    B=ScalarParameter() {defaultValue=1, group=Color, parameterName=IntensityColor_4, sortPriority=8}
   )) {name=Color 4, nodeColor=<rgba 0 0 0 1>}
  NamedRerouteDeclaration(None=MaterialFunctionCall(Size=Constant() {r=1}, CornerRoundness=ScalarParameter() {defaultValue=0.2, group=Outline, parameterName=CornerRadius, sortPriority=2}) {materialFunction=&MF_UI_Box}.Fill) {name=Mask, nodeColor=<rgba 1 0 1 1>}
  NamedRerouteDeclaration(None=Divide(A=ScalarParameter() {defaultValue=1, group=Outline, parameterName=OutlineThickness, sortPriority=1}) {constB=10}) {name=Outline Thickness, nodeColor=<rgba 0 0 0 1>}
  n21 = Constant() {r=1}
  n22 = ScalarParameter() {defaultValue=0.2, group=Outline, parameterName=CornerRadius, sortPriority=2}
  n23 = NamedRerouteUsage()
  NamedRerouteDeclaration(None=OneMinus(None=MaterialFunctionCall(Size=@n21, CornerRoundness=@n22, OutlineThickness=@n23) {materialFunction=&MF_UI_Box}.Outline)) {name=Outline, nodeColor=<rgba 0 0 0 1>}
  n24 = Subtract(A=NamedRerouteUsage()) {constB=0.01}
  n25 = TextureSampleParameter2D(UVs=MaterialFunctionCall(UVs=TextureCoordinate(), Texture Scale=OneMinus(None=@n24)) {materialFunction=&ScaleUVsByCenter}.UVs) {
   group=Texture
   parameterName=Texture
   samplerSource=SSM_Clamp_WorldGroupSettings
   sortPriority=1
   texture=/Game/Athena/UI/InGame/Sidekick/Reactions/100
  }
  NamedRerouteDeclaration(None=@n25.RGB) {name=Texture, nodeColor=<rgba 0 0 0 1>}
  NamedRerouteUsage()
  NamedRerouteUsage()
 ]
}
